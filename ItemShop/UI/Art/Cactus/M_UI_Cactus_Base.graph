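# Animated UI-domain material for a Cactus-themed Item Shop look, with drifting confetti over a colored background.
# The base adds scrolling Snowfall textures, faded by a linear gradient, to a two-color blend.
# Two panning confetti layers share noise distortion, with custom integer-hash nodes randomly masking tiles in each.
# A panning cactus movement texture tints the confetti and, with a second gradient, masks its blend over the base.
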Material /ItemShop/UI/Art/Cactus/M_UI_Cactus_Base
graph {
 properties {bUsedWithStaticMesh=false, materialDomain=MD_UI, pixelDepthOffsetMode=PDOM_Legacy, usedShadingModels=Default Lit}
 nodes [
  n1 = Panner(Coordinate=NamedRerouteUsage(), Time=Time()) {speedX=0.008, speedY=-0.024}
  n2 = Constant2Vector() {g=3, r=3}
  n3 = LinearInterpolate(
   B=TextureSample(UVs=Multiply(A=@n1, B=@n2) {constA=2}) {texture=/Game/UI/Frontend/ItemShop/Art/Snowfall/Textures/T_UI_SnowfallDots}.RGB
   Alpha=MaterialFunctionCall(GradientEnd=Constant() {r=0.75}, RotationAngle=Constant() {r=0.25}, GradientOffset=Constant() {r=0.25}) {materialFunction=&MF_UI_LinearGradient}.Result
  )
  n4 = Multiply(A=Time(), B=ScalarParameter() {defaultValue=0.3616, parameterName=Time factor})
  n5 = TextureSample(UVs=Multiply(A=Panner(Time=@n4) {speedX=0.1079, speedY=0.096}, B=ScalarParameter() {defaultValue=6.6466, parameterName=Param_1})) {texture=/ItemShop/UI/Art/Snowfall/Textures/T_UI_SnowfallDragonMovement}
  n6 = Constant3Vector() {constant=<rgba 0.1285 0.0035 0.5104 0>}
  n7 = Constant3Vector() {constant=<rgba 0.7083 0.1954 0.1254 0>}
  n8 = ComponentMask(None=NamedRerouteUsage()) {g=true}
  n9 = Constant3Vector() {constant=<rgba 0.8854 0.5486 0.5045 0>}
  n10 = Multiply(A=Panner(Coordinate=NamedRerouteUsage(), Time=Time()) {speedX=0.048, speedY=-0.098}, B=Constant2Vector() {g=8, r=8}) {constA=2}
  n11 = Panner(
   Coordinate=Multiply(A=TextureCoordinate(), B=ScalarParameter() {defaultValue=1.1723, parameterName=DupeScaler})
   Time=Multiply(A=Time(), B=ScalarParameter() {defaultValue=0.6346, parameterName=TimeFactor})
  ) {speedX=0.058, speedY=0.096}
  n12 = ScalarParameter() {defaultValue=0.8088, parameterName=ConfettiParam2}
  n13 = Multiply(
   A=TextureSample(UVs=Power(Base=@n11, Exp=@n12)) {samplerType=SAMPLERTYPE_Grayscale, texture=&LowResBlurredNoise}.R
   B=ScalarParameter() {defaultValue=0.752, parameterName=DistortionScaler}
  )
  n14 = Add(A=@n10, B=@n13)
  n15 = Saturate(None=Multiply(A=NamedRerouteUsage(), B=TextureSample(UVs=@n14) {texture=/ItemShop/UI/Art/Cactus/Textures/T_UI_Confetti}.G))
  n16 = Constant3Vector() {constant=<rgba 0.132 0.3179 0.6406 0>}
  n17 = Constant() {r=2}
  n18 = Time()
  n19 = ScalarParameter() {defaultValue=0.3616, parameterName=Time factor}
  n20 = TextureSample(UVs=Panner(
    Coordinate=MaterialFunctionCall(Texture Scale=@n17) {materialFunction=&ScaleUVsByCenter}.UVs
    Time=Multiply(A=@n18, B=@n19)
   ) {speedX=0.1, speedY=-0.272}) {texture=/ItemShop/UI/Art/Cactus/Textures/T_UI_CactusMovement}
  n21 = Multiply(A=Constant3Vector() {constant=<rgba 0.0535 0.5326 0.7344 0>}, B=@n20.R)
  n22 = Time()
  n23 = ScalarParameter() {defaultValue=1, parameterName=Time2}
  n24 = Multiply(A=Panner(Coordinate=NamedRerouteUsage(), Time=Multiply(A=@n22, B=@n23)) {speedX=0.05, speedY=-0.054}, B=Constant2Vector() {g=6, r=6}) {constA=2}
  n25 = Add(A=@n24, B=@n13)
  n26 = Saturate(None=Multiply(A=NamedRerouteUsage(), B=TextureSample(UVs=@n25) {texture=&T_UI_Confetti}.R))
  n27 = Saturate(None=Add(A=@n26, B=@n15))
  n28 = MaterialFunctionCall(GradientEnd=Constant() {r=0.634}, RotationAngle=Constant() {r=0.25}, GradientOffset=Constant() {r=0.3977}) {materialFunction=&MF_UI_LinearGradient}
  LinearInterpolate(
   A=Add(A=LinearInterpolate(B=@n3, Alpha=@n5.RGB), B=LinearInterpolate(A=@n6, B=@n7, Alpha=@n8))
   B=LinearInterpolate(A=Multiply(A=@n9, B=@n15), B=Add(A=@n16, B=@n21), Alpha=@n26)
   Alpha=Multiply(A=@n20.R, B=LinearInterpolate(B=@n27, Alpha=@n28.Result) {constB=0})
  ) {constAlpha=1}
  MaterialFunctionCall() {materialFunction=&ScreenAlignedUVs}
  n29 = MaterialFunctionCall() {materialFunction=&GetUserInterfaceUV}
  n30 = MaterialFunctionCall(Float2=@n29.Pixel Size) {materialFunction=&BreakOutFloat2Components}
  n31 = Multiply(A=Divide(A=@n30.G, B=@n30.G)) {constB=2}
  n32 = Constant() {r=1}
  NamedRerouteDeclaration(None=MaterialFunctionCall(UVs=@n29.Normalized UV, Texture Scale=AppendVector(A=@n31, B=@n32)) {materialFunction=&ScaleUVsByCenter}.UVs) {name=UVCoordinates, nodeColor=<rgba 0 1 0.5312 1>}
  n33 = Constant() {r=20}
  n34 = NamedRerouteUsage()
  n35 = Frac(None=@n34)
  n36 = Multiply(A=@n35, B=OneMinus(None=@n35))
  n37 = Multiply(A=ComponentMask(None=@n36) {r=true}, B=ComponentMask(None=@n36) {g=true})
  MaterialExpressionFloor_0 = Floor(None=@n34)
  n38 = Custom!(x=@MaterialExpressionFloor_0) {
   code="uint2 q = 1103515245U * ( (asuint(x)>>1U) ^ (asuint((x.yx))   ) );\r\nuint  n = 1103515245U * ( (q.x  ) ^ (q.y>>3U) );\r\nreturn float(n) * (1.0/float(0xffffffffU));"
   desc=This is Adrienne Hack
   outputType=CMOT_Float1
   inputs [{inputName=x, input {expression=&self:MaterialExpressionFloor_0}}]
  }
  n39 = Constant() {r=0.67}
  NamedRerouteDeclaration(None=Multiply(A=Multiply(A=@n33, B=@n37), B=Step(Y=@n38, X=@n39))) {name=RandomMasking, nodeColor=<rgba 0 0 0 1>}
  NamedRerouteDeclaration(None=@n24) {name=UVPattern, nodeColor=<rgba 0 0 0 1>}
  NamedRerouteDeclaration(None=@n10) {name=SmallUVPattern, nodeColor=<rgba 0 0 0 1>}
  n40 = Constant() {r=35.997}
  n41 = NamedRerouteUsage()
  n42 = Frac(None=@n41)
  n43 = Multiply(A=@n42, B=OneMinus(None=@n42))
  n44 = Multiply(A=ComponentMask(None=@n43) {r=true}, B=ComponentMask(None=@n43) {g=true})
  MaterialExpressionFloor_3 = Floor(None=@n41)
  n45 = Custom(x=@MaterialExpressionFloor_3) {
   code="uint2 q = 1103515245U * ( (asuint(x)>>1U) ^ (asuint((x.yx))   ) );\r\nuint  n = 1103515245U * ( (q.x  ) ^ (q.y>>3U) );\r\nreturn float(n) * (1.0/float(0xffffffffU));"
   desc=This is Adrienne Hack
   outputType=CMOT_Float1
   inputs [{inputName=x, input {expression=&self:MaterialExpressionFloor_3}}]
  }
  n46 = Constant() {r=0.75}
  NamedRerouteDeclaration(None=Multiply(A=Multiply(A=@n40, B=@n44), B=Step(Y=@n45, X=@n46))) {name=SecondRandomMask, nodeColor=<rgba 0 0 0 1>}
 ]
}
