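# UI material for the Item Shop "Cactus" art that builds an animated confetti backdrop over a two-colour radial gradient.
# Four MF_UI_Confetti layers, each with its own seed, tiling and speed, are masked by vertical UV gradients and blended into the colour lerp.
# Aspect-corrected UV and panning-noise distortion are declared as named reroutes.
# Two inline copies of confetti logic (hashed cells, rotating parallelogram SDF) sit outside that colour chain.
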Material /ItemShop/UI/Art/Cactus/M_UI_Cactus_Base_01
graph {
 properties {bUsedWithStaticMesh=false, materialDomain=MD_UI, pixelDepthOffsetMode=PDOM_Legacy, usedShadingModels=Default Lit}
 nodes [
  n1 = ScalarParameter() {defaultValue=1, parameterName=SmoothStep-Max}
  n2 = ScalarParameter() {defaultValue=0.5, parameterName=SmoothStep-Min}
  n3 = NamedRerouteUsage()
  n4 = NamedRerouteUsage()
  n5 = Multiply(
   A=MaterialFunctionCall(UVs=Add(A=@n3, B=@n4), "Rotation Angle (0-1)"=ScalarParameter() {defaultValue=-0.125, parameterName=Global Angle}) {materialFunction=&CustomRotator}.Rotated Values
   B=AppendVector(A=ScalarParameter() {defaultValue=40, parameterName=Pattern-X}, B=ScalarParameter() {defaultValue=0.75, parameterName=Pattern-Y})
  )
  n6 = Multiply(A=NamedRerouteUsage(), B=Multiply(A=Time(), B=ScalarParameter() {defaultValue=-0.05, parameterName=Confetti-Speed-TimeFactor}))
  n7 = AppendVector(A=Constant(), B=ConstantBiasScale(None=Frac(None=@n6)) {bias=-0.5, scale=2})
  n8 = MaterialFunctionCall(UVs=Add(A=Frac(None=@n5), B=@n7), Texture Scale=NamedRerouteUsage()) {materialFunction=&ScaleUVsByCenter}
  n9 = ScalarParameter() {defaultValue=0.0075, parameterName=Confetti-Local-Rotation-Range}
  n10 = Multiply(A=Time(), B=ScalarParameter() {defaultValue=0.25, parameterName=Confetti-Rotation-TimeFactor})
  n11 = NamedRerouteUsage()
  n12 = LinearInterpolate(A=Multiply(A=@n9) {constB=-1}, B=@n9, Alpha=Sine(None=Multiply(A=@n10, B=@n11))) {constA=0.001, constB=0.001}
  n13 = ScalarParameter() {defaultValue=1, parameterName=Confetti-Local-Tile-X}
  n14 = ScalarParameter() {defaultValue=50, parameterName=Confetti-Local-Tile-Y}
  MaterialExpressionAdd_14 = Add(
   A=Multiply(A=Rotator(Coordinate=@n8.UVs, Time=@n12) {speed=1}, B=AppendVector(A=@n13, B=@n14))
   B=ScalarParameter() {defaultValue=-0.5, parameterName=p}
  )
  MaterialExpressionScalarParameter_47 = ScalarParameter() {defaultValue=0.4, parameterName=wi}
  MaterialExpressionScalarParameter_48 = ScalarParameter() {parameterName=sk}
  MaterialExpressionScalarParameter_49 = ScalarParameter() {defaultValue=0.75, parameterName=he}
  n15 = Custom!(
   p=@MaterialExpressionAdd_14
   wi=@MaterialExpressionScalarParameter_47
   sk=@MaterialExpressionScalarParameter_48
   he=@MaterialExpressionScalarParameter_49
  ) {
   code="float2  e  = float2(sk,he);\r\nfloat e2 = sk*sk + he*he;\r\np = (p.y<0.0)?-p:p;\r\nfloat2  w = p - e; w.x -= clamp(w.x,-wi,wi);\r\nfloat2  d = float2(dot(w,w), -w.y);\r\nfloat s = p.x*e.y - p.y*e.x;\r\np = (s<0.0)?-p:p;\r\nfloat2  v = p - float2(wi,0); v -= e*clamp(dot(v,e)/e2,-1.0,1.0);\r\nd = min( d, float2(dot(v,v), wi*he-abs(s)));\r\nreturn sqrt(d.x)*sign(-d.y);"
   description=Parallelogram
   outputType=CMOT_Float2
   inputs [
    {inputName=p, input {expression=&self:MaterialExpressionAdd_14}}
    {inputName=wi, input {expression=&self:MaterialExpressionScalarParameter_47}}
    {inputName=sk, input {expression=&self:MaterialExpressionScalarParameter_48}}
    {inputName=he, input {expression=&self:MaterialExpressionScalarParameter_49}}
   ]
  }
  n16 = OneMinus(None=SmoothStep(Min=Subtract(A=@n1, B=@n2) {constB=0.25}, Max=@n1, Value=@n15) {constMax=0.6, constMin=0.5, constValue=0.5})
  ComponentMask(None=@n16) {r=true}
  n17 = NamedRerouteUsage()
  n18 = AppendVector(A=ScalarParameter() {defaultValue=0.6, parameterName=Radial-X-Tile}, B=ScalarParameter() {defaultValue=0.95, parameterName=Radial-Y-Tile})
  n19 = Constant2Vector() {g=1, r=0.5}
  n20 = Power(
   Base=Distance(A=Multiply(A=@n17, B=@n18), B=Multiply(A=@n18, B=@n19))
   Exp=ScalarParameter() {defaultValue=1.25, parameterName=Distance Power}
  )
  n21 = NamedRerouteUsage()
  n22 = NamedRerouteUsage()
  n23 = ScalarParameter() {parameterName=L1-Patter-X-Offset}
  n24 = ScalarParameter() {defaultValue=0.5, parameterName=L1-Patter-Y-Offset}
  n25 = MaterialFunctionCall!(
   "UV(V2)"=Add(A=Add(A=@n21, B=@n22), B=AppendVector(A=@n23, B=@n24))
   Global-Angle=ScalarParameter() {defaultValue=-0.125, parameterName=L1-Global Angle}
   "Pattern-Tiles(V2)"=AppendVector(A=ScalarParameter() {defaultValue=50, parameterName=L1-Pattern-X}, B=ScalarParameter() {defaultValue=0.75, parameterName=L1-Pattern-Y})
   RandomSeed=ScalarParameter() {defaultValue=12345, parameterName=L1-RandomSeed}
   Confetti-Speed-TimeFactor=ScalarParameter() {defaultValue=-0.05, parameterName=L1-Confetti-Speed-TimeFactor}
   Confetti-Rotation-TimeFactor=ScalarParameter() {defaultValue=0.25, parameterName=L1-Confetti-Rotation-TimeFactor}
   Confetti-Local-Rotation-Range=ScalarParameter() {defaultValue=0.0075, parameterName=L1-Confetti-Local-Rotation-Range}
   "Confetti-Local-Tiles(V2)"=AppendVector(
    A=ScalarParameter() {defaultValue=1, parameterName=L1-Confetti-Local-Tile-X}
    B=ScalarParameter() {defaultValue=50, parameterName=L1-Confetti-Local-Tile-Y}
   )
   "SmoothStepRanges(V2)"=AppendVector(
    A=ScalarParameter() {defaultValue=0.9, parameterName=L1-SmoothStep-Max}
    B=ScalarParameter() {defaultValue=0.5, parameterName=L1-SmoothStep-Min}
   )
  ) {materialFunction=/Game/UI/Frontend/ItemShop/Art/Cactus/MF_UI_Confetti}
  n26 = NamedRerouteUsage()
  n27 = NamedRerouteUsage()
  n28 = ScalarParameter() {parameterName=L2-Patter-X-Offset}
  n29 = ScalarParameter() {defaultValue=0.15, parameterName=L2-Patter-Y-Offset}
  n30 = MaterialFunctionCall(
   "UV(V2)"=Add(A=Add(A=@n26, B=@n27), B=AppendVector(A=@n28, B=@n29))
   Global-Angle=ScalarParameter() {defaultValue=-0.075, parameterName=L2-Global Angle}
   "Pattern-Tiles(V2)"=AppendVector(A=ScalarParameter() {defaultValue=60, parameterName=L2-Pattern-X}, B=ScalarParameter() {defaultValue=2, parameterName=L2-Pattern-Y})
   RandomSeed=ScalarParameter() {defaultValue=2345, parameterName=L2-RandomSeed}
   Confetti-Speed-TimeFactor=ScalarParameter() {defaultValue=-0.05, parameterName=L2-Confetti-Speed-TimeFactor}
   Confetti-Rotation-TimeFactor=ScalarParameter() {defaultValue=0.125, parameterName=L2-Confetti-Rotation-TimeFactor}
   Confetti-Local-Rotation-Range=ScalarParameter() {defaultValue=0.005, parameterName=L2-Confetti-Local-Rotation-Range}
   "Confetti-Local-Tiles(V2)"=AppendVector(
    A=ScalarParameter() {defaultValue=1, parameterName=L2-Confetti-Local-Tile-X}
    B=ScalarParameter() {defaultValue=37.1648, parameterName=L2-Confetti-Local-Tile-Y}
   )
   "SmoothStepRanges(V2)"=AppendVector(
    A=ScalarParameter() {defaultValue=1, parameterName=L2-SmoothStep-Max}
    B=ScalarParameter() {defaultValue=0.5, parameterName=L2-SmoothStep-Min}
   )
  ) {materialFunction=&MF_UI_Confetti}
  n31 = MaterialFunctionCall(
   "UV(V2)"=Add(A=NamedRerouteUsage(), B=NamedRerouteUsage())
   Global-Angle=ScalarParameter() {defaultValue=-0.075, parameterName=L4-Global Angle}
   "Pattern-Tiles(V2)"=AppendVector(A=ScalarParameter() {defaultValue=89.0677, parameterName=L4-Pattern-X}, B=ScalarParameter() {defaultValue=7, parameterName=L4-Pattern-Y})
   RandomSeed=ScalarParameter() {defaultValue=3863.1282, parameterName=L4-RandomSeed}
   Confetti-Speed-TimeFactor=ScalarParameter() {defaultValue=-0.171, parameterName=L4-Confetti-Speed-TimeFactor}
   Confetti-Rotation-TimeFactor=ScalarParameter() {defaultValue=0.101, parameterName=L4-Confetti-Rotation-TimeFactor}
   Confetti-Local-Rotation-Range=ScalarParameter() {defaultValue=0.005, parameterName=L4-Confetti-Local-Rotation-Range}
   "Confetti-Local-Tiles(V2)"=AppendVector(
    A=ScalarParameter() {defaultValue=1, parameterName=L4-Confetti-Local-Tile-X}
    B=ScalarParameter() {defaultValue=33.2071, parameterName=L4-Confetti-Local-Tile-Y}
   )
   "SmoothStepRanges(V2)"=AppendVector(
    A=ScalarParameter() {defaultValue=-0.075, parameterName=L4-SmoothStep-Max}
    B=ScalarParameter() {defaultValue=0.484, parameterName=L4-SmoothStep-Min}
   )
  ) {materialFunction=&MF_UI_Confetti}
  n32 = ComponentMask(None=TextureCoordinate()) {g=true}
  n33 = Saturate(None=Subtract(A=OneMinus(None=@n32), B=ScalarParameter() {defaultValue=0.25, parameterName=L4-GradientMask}))
  n34 = MaterialFunctionCall(
   "UV(V2)"=Add(A=NamedRerouteUsage(), B=NamedRerouteUsage())
   Global-Angle=ScalarParameter() {defaultValue=-0.075, parameterName=L3-Global Angle}
   "Pattern-Tiles(V2)"=AppendVector(A=ScalarParameter() {defaultValue=60.48, parameterName=L3-Pattern-X}, B=ScalarParameter() {defaultValue=5, parameterName=L3-Pattern-Y})
   RandomSeed=ScalarParameter() {defaultValue=3456, parameterName=L3-RandomSeed}
   Confetti-Speed-TimeFactor=ScalarParameter() {defaultValue=-0.091, parameterName=L3-Confetti-Speed-TimeFactor}
   Confetti-Rotation-TimeFactor=ScalarParameter() {defaultValue=0.013, parameterName=L3-Confetti-Rotation-TimeFactor}
   Confetti-Local-Rotation-Range=ScalarParameter() {defaultValue=0.005, parameterName=L3-Confetti-Local-Rotation-Range}
   "Confetti-Local-Tiles(V2)"=AppendVector(
    A=ScalarParameter() {defaultValue=1, parameterName=L3-Confetti-Local-Tile-X}
    B=ScalarParameter() {defaultValue=29.76, parameterName=L3-Confetti-Local-Tile-Y}
   )
   "SmoothStepRanges(V2)"=AppendVector(
    A=ScalarParameter() {defaultValue=0.109, parameterName=L3-SmoothStep-Max}
    B=ScalarParameter() {defaultValue=0.5, parameterName=L3-SmoothStep-Min}
   )
  ) {materialFunction=&MF_UI_Confetti}
  n35 = ComponentMask(None=TextureCoordinate()) {g=true}
  n36 = Saturate(None=Subtract(A=OneMinus(None=@n35), B=ScalarParameter() {defaultValue=0.25, parameterName=L3-GradientMask}))
  n37 = LinearInterpolate(
   A=Multiply(A=Multiply(A=@n31.Confetti-Mask, B=@n33), B=ScalarParameter() {defaultValue=0.75, parameterName=L4-Intensity})
   Alpha=Multiply(A=Multiply(A=@n34.Confetti-Mask, B=@n36), B=ScalarParameter() {defaultValue=0.75, parameterName=L3-Intensity})
  )
  n38 = TextureCoordinate()
  n39 = Subtract(A=OneMinus(None=ComponentMask(None=@n38) {g=true}), B=ScalarParameter() {defaultValue=0.35, parameterName=L2-GradientMask})
  n40 = Multiply(A=Multiply(A=@n30.Confetti-Mask, B=Saturate(None=@n39)), B=ScalarParameter() {defaultValue=0.75, parameterName=L2-Intensity})
  n41 = TextureCoordinate()
  n42 = Subtract(A=OneMinus(None=ComponentMask(None=@n41) {g=true}), B=ScalarParameter() {defaultValue=0.3, parameterName=L1-Gradient-Mask})
  n43 = ScalarParameter() {defaultValue=2, parameterName=L1-Gradient-Falloff}
  n44 = Multiply(A=@n25.Confetti-Mask, B=Saturate(None=Multiply(A=@n42, B=@n43)))
  n45 = ScalarParameter() {defaultValue=1, parameterName=L1-Intensity}
  LinearInterpolate(
   A=LinearInterpolate(
    A=Constant3Vector() {constant=<rgba 0.7083 0.1954 0.1254 0>}
    B=Constant3Vector() {constant=<rgba 0.1285 0.0035 0.5104 0>}
    Alpha=Saturate(None=@n20)
   )
   B=LinearInterpolate(
    A=Constant3Vector() {constant=<rgba 0.8854 0.5486 0.5045 0>}
    B=Constant3Vector() {constant=<rgba 0.0535 0.5326 0.7344 0>}
    Alpha=LinearInterpolate(A=@n30.Confetti-Mask, B=@n25.Confetti-Mask, Alpha=@n25.Confetti-Mask)
   )
   Alpha=LinearInterpolate(A=LinearInterpolate(A=@n37, Alpha=@n40), Alpha=Multiply(A=@n44, B=@n45))
  )
  n46 = ScalarParameter() {defaultValue=1, parameterName=SmoothStep-Max}
  n47 = ScalarParameter() {defaultValue=0.5, parameterName=SmoothStep-Min}
  n48 = NamedRerouteUsage()
  n49 = NamedRerouteUsage()
  n50 = Multiply(
   A=MaterialFunctionCall(UVs=Add(A=@n48, B=@n49), "Rotation Angle (0-1)"=ScalarParameter() {defaultValue=-0.125, parameterName=Global Angle}) {materialFunction=&CustomRotator}.Rotated Values
   B=AppendVector(A=ScalarParameter() {defaultValue=40, parameterName=Pattern-X}, B=ScalarParameter() {defaultValue=0.75, parameterName=Pattern-Y})
  )
  n51 = Multiply(A=NamedRerouteUsage(), B=Multiply(A=Time(), B=ScalarParameter() {defaultValue=-0.05, parameterName=Confetti-Speed-TimeFactor}))
  n52 = AppendVector(A=Constant(), B=ConstantBiasScale(None=Frac(None=@n51)) {bias=-0.5, scale=2})
  n53 = MaterialFunctionCall(UVs=Add(A=Frac(None=@n50), B=@n52), Texture Scale=NamedRerouteUsage()) {materialFunction=&ScaleUVsByCenter}
  n54 = ScalarParameter() {defaultValue=0.0075, parameterName=Confetti-Local-Rotation-Range}
  n55 = Multiply(A=Time(), B=ScalarParameter() {defaultValue=0.25, parameterName=Confetti-Rotation-TimeFactor})
  n56 = NamedRerouteUsage()
  n57 = LinearInterpolate(A=Multiply(A=@n54) {constB=-1}, B=@n54, Alpha=Sine(None=Multiply(A=@n55, B=@n56))) {constA=0.001, constB=0.001}
  n58 = ScalarParameter() {defaultValue=1, parameterName=Confetti-Local-Tile-X}
  n59 = ScalarParameter() {defaultValue=50, parameterName=Confetti-Local-Tile-Y}
  MaterialExpressionAdd_8 = Add(
   A=Multiply(A=Rotator(Coordinate=@n53.UVs, Time=@n57) {speed=1}, B=AppendVector(A=@n58, B=@n59))
   B=ScalarParameter() {defaultValue=-0.5, parameterName=p}
  )
  MaterialExpressionScalarParameter_21 = ScalarParameter() {defaultValue=0.4, parameterName=wi}
  MaterialExpressionScalarParameter_22 = ScalarParameter() {parameterName=sk}
  MaterialExpressionScalarParameter_23 = ScalarParameter() {defaultValue=0.75, parameterName=he}
  n60 = Custom(
   p=@MaterialExpressionAdd_8
   wi=@MaterialExpressionScalarParameter_21
   sk=@MaterialExpressionScalarParameter_22
   he=@MaterialExpressionScalarParameter_23
  ) {
   code="float2  e  = float2(sk,he);\r\nfloat e2 = sk*sk + he*he;\r\np = (p.y<0.0)?-p:p;\r\nfloat2  w = p - e; w.x -= clamp(w.x,-wi,wi);\r\nfloat2  d = float2(dot(w,w), -w.y);\r\nfloat s = p.x*e.y - p.y*e.x;\r\np = (s<0.0)?-p:p;\r\nfloat2  v = p - float2(wi,0); v -= e*clamp(dot(v,e)/e2,-1.0,1.0);\r\nd = min( d, float2(dot(v,v), wi*he-abs(s)));\r\nreturn sqrt(d.x)*sign(-d.y);"
   description=Parallelogram
   outputType=CMOT_Float2
   inputs [
    {inputName=p, input {expression=&self:MaterialExpressionAdd_8}}
    {inputName=wi, input {expression=&self:MaterialExpressionScalarParameter_21}}
    {inputName=sk, input {expression=&self:MaterialExpressionScalarParameter_22}}
    {inputName=he, input {expression=&self:MaterialExpressionScalarParameter_23}}
   ]
  }
  n61 = OneMinus(None=SmoothStep(Min=Subtract(A=@n46, B=@n47) {constB=0.25}, Max=@n46, Value=@n60) {constMax=0.6, constMin=0.5, constValue=0.5})
  n62 = TextureCoordinate()
  Multiply(A=ComponentMask(None=@n61) {r=true}, B=OneMinus(None=ComponentMask(None=@n62) {g=true}))
  Multiply(A=@n61, B=@n52)
  Multiply(A=@n16, B=@n7)
  n63 = MaterialFunctionCall() {materialFunction=&GetUserInterfaceUV}
  n64 = MaterialFunctionCall(Float2=@n63.Pixel Size) {materialFunction=&BreakOutFloat2Components}
  n65 = AppendVector(A=Divide(A=@n64.R, B=Max(A=@n64.G)), B=Constant() {r=1})
  n66 = OneMinus(None=@n65)
  NamedRerouteDeclaration(None=Add(A=Multiply(A=@n66) {constB=0.5}, B=Multiply(A=@n65, B=@n63.Normalized UV))) {name="Squared UV - (x/y,y)", nodeColor=<rgba 0 0 0 1>}
  n67 = TextureCoordinate()
  n68 = ScalarParameter() {defaultValue=0.5, parameterName=DupeScaler}
  n69 = Time()
  n70 = ScalarParameter() {defaultValue=0.6346, parameterName=TimeFactor}
  n71 = Power(
   Base=Panner(Coordinate=Multiply(A=@n67, B=@n68), Time=Multiply(A=@n69, B=@n70)) {speedX=0.058, speedY=0.096}
   Exp=ScalarParameter() {defaultValue=0.8088, parameterName=ConfettiParam2}
  )
  NamedRerouteDeclaration(None=Multiply(
    A=TextureSample(UVs=@n71) {samplerType=SAMPLERTYPE_Grayscale, texture=&LowResBlurredNoise}.R
    B=ScalarParameter() {defaultValue=0.35, parameterName=DistortionScaler}
   )) {name=UV-Distortion, nodeColor=<rgba 0 0 0 1>}
  MaterialExpressionAdd_11 = Add(A=ScalarParameter() {defaultValue=1453, parameterName=RandomSeed}, B=Add(A=Floor(None=@n5)) {constB=0.01})
  n72 = Custom!(x=@MaterialExpressionAdd_11) {
   code="uint2 q = 1103515245U * ( (asuint(x)>>1U) ^ (asuint((x.yx))   ) );\r\nuint  n = 1103515245U * ( (q.x  ) ^ (q.y>>3U) );\r\nreturn float(n) * (1.0/float(0xffffffffU));"
   desc=Random
   description=IQHash3
   outputType=CMOT_Float1
   inputs [{inputName=x, input {expression=&self:MaterialExpressionAdd_11}}]
  }
  NamedRerouteDeclaration(None=@n72) {name=Random_1, nodeColor=<rgba 0 0 0 1>}
  MaterialExpressionAdd_6 = Add(A=ScalarParameter() {defaultValue=1453, parameterName=RandomSeed}, B=Add(A=Floor(None=@n50)) {constB=0.01})
  n73 = Custom(x=@MaterialExpressionAdd_6) {
   code="uint2 q = 1103515245U * ( (asuint(x)>>1U) ^ (asuint((x.yx))   ) );\r\nuint  n = 1103515245U * ( (q.x  ) ^ (q.y>>3U) );\r\nreturn float(n) * (1.0/float(0xffffffffU));"
   desc=Random
   description=IQHash3
   outputType=CMOT_Float1
   inputs [{inputName=x, input {expression=&self:MaterialExpressionAdd_6}}]
  }
  NamedRerouteDeclaration(None=@n73) {name=Random, nodeColor=<rgba 0 0 0 1>}
 ]
}
